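# Animation Blueprint for the F_MED_Temple_Throwback female body that corrects thigh holster and glute bones as the legs rotate.
# It copies the attached parent's pose, blends each holster bone with weighted constraints to pelvis, thigh and thigh twist, then adds offsets to holster and glute bones.
# Each update derives those offset alphas from thigh rotation relative to the pelvis.
# Backbling and head-look alphas appear in the event graph but drive nothing in the AnimGraph.
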AnimBlueprint /BRCosmetics/Characters/Player/Female/Medium/Bodies/F_MED_Temple_Throwback/Meshes/F_MED_Temple_Throwback_AnimBP
target_skeleton=&/F_MED_Temple_Throwback_Skeleton
graphs [
 AnimGraph {nodes [
   n1 = "Constraint - dyn_parent_thigh_holster_l (PPP)"!(
    ComponentPose=Local To Component(LocalPose=Copy Pose From Mesh() {bUseAttachedParent=true})
    Alpha=1
    ConstraintWeights_0=0.2195
    ConstraintWeights_1:float=0.3198
    ConstraintWeights_2:float=0.4608
   ) -> (Pose) {
    alphaBoolBlend {blendOption=Linear}
    boneToModify {boneName=dyn_parent_thigh_holster_l}
    constraintSetup [{targetBone {boneName=pelvis}}, {targetBone {boneName=thigh_twist_01_l}}, {targetBone {boneName=thigh_l}}]
    constraintWeights [1,1,1]
   }
   n2 = "Constraint - dyn_parent_thigh_holster_r (PPP)"!(ComponentPose=@n1, Alpha=1, ConstraintWeights_0=0.4608, ConstraintWeights_1:float=0.2195, ConstraintWeights_2:float=0.3198) -> (Pose) {
    alphaBoolBlend {blendOption=Linear}
    boneToModify {boneName=dyn_parent_thigh_holster_r}
    constraintSetup [{targetBone {boneName=thigh_r}}, {targetBone {boneName=pelvis}}, {targetBone {boneName=thigh_twist_01_r}}]
    constraintWeights [1,1,1]
   }
   n3 = "Transform (Modify) Bone - Bone: dyn_parent_thigh_holster_l"(ComponentPose=@n2, Alpha=Get thigh_rot_l(), Translation="-2,0,3.5", Rotation="9,0,0", Scale="1,1,1") {
    rotationMode=BMM_Additive
    rotationSpace=BCS_BoneSpace
    translationMode=BMM_Additive
    translationSpace=BCS_BoneSpace
    alphaBoolBlend {blendOption=Linear}
    boneToModify {boneName=dyn_parent_thigh_holster_l}
    rotation {pitch=9}
    translation {x=-2, z=3.5}
   }
   n4 = "Transform (Modify) Bone - Bone: dyn_parent_thigh_holster_r"(ComponentPose=@n3, Alpha=Get thigh_rot_r(), Translation="-2,0,-3.5", Rotation="7,0,0", Scale="1,1,1") {
    rotationMode=BMM_Additive
    rotationSpace=BCS_BoneSpace
    translationMode=BMM_Additive
    translationSpace=BCS_BoneSpace
    alphaBoolBlend {blendOption=Linear}
    boneToModify {boneName=dyn_parent_thigh_holster_r}
    rotation {pitch=7}
    translation {x=-2, z=-3.5}
   }
   n5 = "Transform (Modify) Bone - Bone: dyn_glute_r"(ComponentPose=@n4, Alpha=Get thigh_rot_r(), Translation="5,-5,-7", Rotation="0.000,0.000,-35", Scale="1.3,1.3,1.3") {
    rotationMode=BMM_Additive
    rotationSpace=BCS_BoneSpace
    scaleMode=BMM_Additive
    scaleSpace=BCS_BoneSpace
    translationMode=BMM_Additive
    translationSpace=BCS_BoneSpace
    alphaBoolBlend {blendOption=Linear}
    boneToModify {boneName=dyn_glute_r}
    rotation {roll=-35}
    scale=<1.3 1.3 1.3>
    translation=<5 -5 -7>
   }
   n6 = "Transform (Modify) Bone - Bone: dyn_glute_l"(ComponentPose=@n5, Alpha=Get thigh_rot_l(), Translation="-5.5,-2,0", Rotation="0,0,-20", Scale="1.2,1.2,1.2") {
    rotationMode=BMM_Additive
    rotationSpace=BCS_BoneSpace
    scaleMode=BMM_Additive
    scaleSpace=BCS_BoneSpace
    translationMode=BMM_Additive
    translationSpace=BCS_BoneSpace
    alphaBoolBlend {blendOption=Linear}
    boneToModify {boneName=dyn_glute_l}
    rotation {roll=-20}
    scale=<1.2 1.2 1.2>
    translation=<-5.5 -2>
   }
   Output Pose(Result=Component To Local(ComponentPose=@n6)) {layerGroup=None}
  ]}
 EventGraph {nodes [
   Event BlueprintUpdateAnimation()
   > Get!(execute:Exec) -> (then:Exec, else:Exec, Skel_Mesh:Skeletal Mesh Component Object Reference)
   > Sequence()
   n1 = Get Skel_Mesh()
   > Set thigh_rot_l(thigh_rot_l=GetAlphaFromReferenceTransformDelta(SkeletalMeshComponent=@n1, BoneName=thigh_l, BaseName=pelvis, InputRange=(X=-20,Y=-100), TwistAxis="1,0,0"))
   > Set thigh_rot_r(thigh_rot_r=GetAlphaFromReferenceTransformDelta(SkeletalMeshComponent=@n1, BoneName=thigh_r, BaseName=pelvis, InputRange=(X=-20,Y=-100), TwistAxis="1,0,0"))
   Event BlueprintInitializeAnimation()
   > Set Skel_Mesh(Skel_Mesh=GetPartSkeletalMeshComponent())
   n2 = CheckForBackbling()
   n3 = Set isWearingBackbling(isWearingBackbling=OR Boolean(A=@n2, B=@n2.bIsBackblingLong))
   > n4 = Branch(Condition=@n3)
   > n5 = Set backpack_alpha(backpack_alpha:double=1) -> (Output_Get:double)
   Get(execute=[@n5.then, @n6.then])
   > Set head delta_look_up(head delta_look_up:double=GetAlphaFromReferenceTransformDelta(
     SkeletalMeshComponent=Get Skel_Mesh()
     BoneName=head
     BaseName=spine_05
     DeltaComponentType=QuaternionTwist
     InputRange=(X=20,Y=46)
     TwistAxis="0,0,1"
    )) -> (Output_Get:double)
   n6 = Set backpack_alpha(execute=@n4.else, backpack_alpha=0)
   Event OnCharacterCustomizationLoaded()
  ]}
]
variables [
 backpack_alpha {type=number}
 head delta_look_dwn {type=number}
 head delta_look_lft {type=number}
 head delta_look_rgt {type=number}
 head delta_look_up {type=number}
 isWearingBackbling {type=boolean}
 leg_rot_up_l {type=number}
 Skel_Mesh {type=SkeletalMeshComponent}
 thigh_rot_l {type=number}
 thigh_rot_r {type=number}
]
