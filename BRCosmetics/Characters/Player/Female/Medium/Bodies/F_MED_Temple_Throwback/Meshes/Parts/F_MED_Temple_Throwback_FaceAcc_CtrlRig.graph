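# Control Rig for the Temple Throwback female face accessory part that shapes the six-bone hair tail along a curve.
# A four-point Bezier is built from the tail's parent bone on the head and control bones 02, 04 and 06 under the neck, left clavicle and spine.
# Fit Chain To Curve lays the tail bones along it with stretched alignment, so the tail spans between head and upper body.
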ControlRigBlueprint /BRCosmetics/Characters/Player/Female/Medium/Bodies/F_MED_Temple_Throwback/Meshes/Parts/F_MED_Temple_Throwback_FaceAcc_CtrlRig
graphs [
 RigVMFunctionLibrary {links [], nodes []}
 RigVMModel {nodes [
   RigUnit_BeginExecution = &RigUnit_BeginExecution()
   RigUnit_GetTransform = &RigUnit_GetTransform(Item=(Type=Bone,Name="dyn_parent_hair_tail_01"))
   RigUnit_CollectionItems = /Script/ControlRig.RigUnit_CollectionItems(Items=((Type=Bone,Name="dyn_hair_tail_01"),(Type=Bone,Name="dyn_hair_tail_02"),(Type=Bone,Name="dyn_hair_tail_03"),(Type=Bone,Name="dyn_hair_tail_04"),(Type=Bone,Name="dyn_hair_tail_05"),(Type=Bone,Name="dyn_hair_tail_06"))) {is=struct RigVMUnitNode}
   /Script/ControlRig.RigUnit_FitChainToCurvePerItem(
    Alignment=Stretched
    DebugSettings=(Scale=1,CurveColor=(R=1,G=1,B=0,A=1),SegmentsColor=(R=1,G=0,B=0,A=1),WorldOffset=(Rotation=(X=0,Y=0,Z=0,W=1),Translation=(X=0,Y=0,Z=0),Scale3D=(X=1,Y=1,Z=1)))
    Maximum=1
    PoleVectorPosition=(X=0,Y=0,Z=0)
    PrimaryAxis=(X=1,Y=0,Z=0)
    RotationEaseType=Linear
    Rotations=()
    SamplingPrecision=12
    SecondaryAxis=(X=0,Y=0,Z=0)
    Weight=1
    bPropagateToChildren=True
    Bezier=/Script/RigVM.RigVMFunction_MathVectorMakeBezierFourPoint(
     Bezier=(A=(X=0,Y=0,Z=0),B=(X=0,Y=0,Z=0),C=(X=0,Y=0,Z=0),D=(X=0,Y=0,Z=0))
     Bezier.A=@RigUnit_GetTransform.Transform.Translation
     Bezier.B=&RigUnit_GetTransform(Item=(Type=Bone,Name="dyn_con_hair_tail_02")).Transform.Translation
     Bezier.C=&RigUnit_GetTransform(Item=(Type=Bone,Name="dyn_con_hair_tail_04")).Transform.Translation
     Bezier.D=&RigUnit_GetTransform(Item=(Type=Bone,Name="dyn_con_hair_tail_06")).Transform.Translation
    ) {is=struct RigVMUnitNode}.Bezier
    ExecutePin=@RigUnit_BeginExecution.ExecutePin
    Items=@RigUnit_CollectionItems.Collection
   ) {is=struct RigVMUnitNode}
  ]}
]
hierarchy [
 root {loc=<0 0 0>, rot=<0 0 0 1>}
 pelvis {parent=root, loc=<0 -0.5301 98.0571>, rot=<0.0127 0.707 -0.0127 -0.707>}
 spine_01 {parent=pelvis, loc=<6.1208 -0.9328 0>, rot=<0 0 0.034 -0.9994>}
 spine_02 {parent=spine_01, loc=<6.5996 -0.2095 0>, rot=<0 0 0.0238 -0.9997>}
 spine_03 {parent=spine_02, loc=<6.4914 -0.1237 0>, rot=<0 0 0.0149 -0.9999>}
 spine_04 {parent=spine_03, loc=<6.9598 0.0794 0>, rot=<0 0 -0.0132 -0.9999>}
 spine_05 {parent=spine_04, loc=<6.9987 0 0>, rot=<0 0 -0.0015 -1>}
 clavicle_l {parent=spine_05, loc=<5.4831 -0.1053 -3.2946>, rot=<0.0465 -0.7289 -0.009 -0.6829>}
 upperarm_l {parent=clavicle_l, loc=<9.8658 -1.2136 -0.4105>, rot=<0.0164 -0.3245 -0.0269 -0.9454>}
 dyn_con_hair_tail_04 {parent=clavicle_l, loc=<3.6072 7.3687 3.0987>, rot=<0.308 -0.2756 -0.6289 -0.6585>}
 dyn_con_hair_tail_03 {parent=clavicle_l, loc=<2.7165 -0.4912 7.3223>, rot=<0.1932 -0.1526 -0.652 -0.7171>}
 clavicle_r {parent=spine_05, loc=<5.4831 -0.1053 3.2946>, rot=<0.7289 0.0465 -0.6829 0.009>}
 upperarm_r {parent=clavicle_r, loc=<-9.8658 1.2136 0.4104>, rot=<0.0164 -0.3245 -0.0269 -0.9454>}
 neck_01 {parent=spine_05, loc=<11.3138 0.1529 0>, rot=<0 0 -0.2003 -0.9797>}
 neck_02 {parent=neck_01, loc=<4.5627 0 0>, rot=<0 0 0.0504 -0.9987>}
 head {parent=neck_02, loc=<5.3914 0 0>, rot=<0 0 0.115 -0.9934>}
 dyn_parent_hair_tail_01 {parent=head, loc=<-1.0965 -8.9003 -0.0126>, rot=<0.4267 -0.9044 0.0009 -0.002>}
 dyn_hair_tail_01 {parent=dyn_parent_hair_tail_01, loc=<0 0 0>, rot=<0 0 -0.1739 -0.9848>}
 dyn_hair_tail_02 {parent=dyn_hair_tail_01, loc=<7.7716 0 0>, rot=<0.0008 -0.0021 -0.1482 -0.989>}
 dyn_hair_tail_03 {parent=dyn_hair_tail_02, loc=<8.1739 0 0>, rot=<0 0 -0.0985 -0.9951>}
 dyn_hair_tail_04 {parent=dyn_hair_tail_03, loc=<8.967 0.0347 0>, rot=<0 0 -0.037 -0.9993>}
 dyn_hair_tail_05 {parent=dyn_hair_tail_04, loc=<10.9582 0 0>, rot=<0 0 0 1>}
 dyn_hair_tail_06 {parent=dyn_hair_tail_05, loc=<7.9156 0 0>, rot=<0 0 0 1>}
 dyn_hair_tip_01 {parent=head, loc=<10.6317 7.0025 -5.2686>, rot=<0.109 -0.9697 -0.0668 -0.2084>}
 dyn_hair_tip_02 {parent=dyn_hair_tip_01, loc=<4.7849 0 0.0746>, rot=<0.0443 -0.1813 -0.0082 -0.9824>}
 dyn_hair_strand_01_l {parent=head, loc=<12.5065 8.2423 0.8153>, rot=<0.0245 -0.7544 -0.0989 -0.6485>}
 dyn_hair_strand_02_l {parent=dyn_hair_strand_01_l, loc=<5.6988 0 0>, rot=<0.0137 -0.481 0.0243 -0.8763>}
 dyn_hair_strand_03_l {parent=dyn_hair_strand_02_l, loc=<4.9802 0 0>, rot=<0.0443 -0.1813 -0.0082 -0.9824>}
 dyn_hair_strand_01_r {parent=head, loc=<11.7011 7.92 3.4801>, rot=<0.34 0.1529 0.9231 -0.0941>}
 dyn_hair_strand_02_r {parent=dyn_hair_strand_01_r, loc=<4.0997 0 0>, rot=<-0.0222 -0.246 -0.1747 -0.9531>}
 dyn_hair_strand_03_r {parent=dyn_hair_strand_02_r, loc=<4.2828 0.416 -0.1101>, rot=<0 0 0 -1>}
 dyn_con_hair_tail_01 {parent=head, loc=<-1.0965 -8.9003 -0.0126>, rot=<0.192 -0.8912 -0.144 -0.3849>}
 dyn_con_hair_tail_02 {parent=neck_01, loc=<0.7996 -8.6381 -5.4818>, rot=<-0.4497 -0.5707 -0.4034 -0.5562>}
 dyn_backpack_col {parent=spine_05, loc=<0 0 0>, rot=<0 0 0 1>}
 dyn_con_hair_tail_05 {parent=spine_04, loc=<5.91 13.7843 -7.9037>, rot=<-0.0695 -0.7278 -0.6784 -0.0729>}
 dyn_con_hair_tail_06 {parent=spine_04, loc=<-1.8451 15.3673 -7.9972>, rot=<0.163 -0.6803 -0.6555 0.2846>}
]
