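# Physics asset for the F_MED_Temple_Throwback female player cosmetic body, set up around a dynamic hair tail.
# Six tapering capsules form the tail, chained by limited-swing, tight-twist constraints from a small parent body that links to the head with free rotation.
# Capsules on the head, spine, pelvis, upper arms and a backpack collision bone form a coarse body shape around the tail.
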PhysicsAsset /BRCosmetics/Characters/Player/Female/Medium/Bodies/F_MED_Temple_Throwback/Meshes/Parts/F_MED_Temple_Throwback_Physics
bodies {
 dyn_backpack_col [{shapeType=Capsule, length=11, radius=13.25, center=<-5.7008 8.2066 0.0002>, rotation=<rot 90 171.1193 167.3347>}]
 dyn_hair_tail_01 [{shapeType=Capsule, length=7.1758, radius=4.7028, center=<7.1932 0 0>, rotation=<rot 90 0 0>}]
 dyn_hair_tail_02 [{shapeType=Capsule, length=4.8934, radius=4.0751, center=<5.2773 0 0>, rotation=<rot 90 0 0>}]
 dyn_hair_tail_03 [{shapeType=Capsule, length=7.1297, radius=3.6405, center=<4.9027 0 0>, rotation=<rot 90 0 0>}]
 dyn_hair_tail_04 [{shapeType=Capsule, length=6.7383, radius=2.9232, center=<4.4606 0 0>, rotation=<rot 90 0 0>}]
 dyn_hair_tail_05 [{shapeType=Capsule, length=4.9242, radius=2.4885, center=<2.2326 0 0>, rotation=<rot 90 0 0>}]
 dyn_hair_tail_06 [{shapeType=Capsule, length=6.1225, radius=2.2491, center=<3.331 0 0>, rotation=<rot 90 0 0>}]
 dyn_parent_hair_tail_01 [{shapeType=Capsule, length=0.505, radius=0.505, rotation=<rot 0 -90 0>}]
 head [{shapeType=Capsule, length=4, radius=9, center=<4.1499 0 0>, rotation=<rot 90 0 0>}]
 pelvis [{shapeType=Capsule, length=9.5976, radius=11.9449, center=<0.7824 -0.1228 -0.8746>}]
 spine_02 [{shapeType=Capsule, length=6.75, radius=10.5, center=<3.5397 2.4742 0>, rotation=<rot 90 0 0>}]
 spine_05 [{shapeType=Capsule, length=2.5, radius=16.1879, center=<-0.0023 8.3414 -0.0233>, rotation=<rot 0.0318 -89.9995 0.0761>}]
 upperarm_l [{shapeType=Capsule, length=21.058, radius=3.9313, center=<12.0777 0.0483 0.1995>, rotation=<rot 0 -92.3883 -89.7622>}]
 upperarm_r [{shapeType=Capsule, length=21.058, radius=3.9313, center=<-12.0777 -0.0483 -0.1995>, rotation=<rot 0 -92.3883 -89.7622>}]
}
constraints [
 {bone1Name=spine_05, bone2Name=spine_02, swing1LimitDegrees=45, swing2LimitDegrees=45, twistLimitDegrees=45}
 {bone1Name=head, bone2Name=spine_05, swing1LimitDegrees=45, swing2LimitDegrees=45, twistLimitDegrees=45}
 {
  bone1Name=spine_02
  bone2Name=pelvis
  swing1LimitDegrees=45
  swing1Motion=Free
  swing2LimitDegrees=45
  swing2Motion=Free
  twistLimitDegrees=45
  twistMotion=Free
 }
 {
  bone1Name=dyn_parent_hair_tail_01
  bone2Name=head
  swing1LimitDegrees=45
  swing1Motion=Free
  swing2LimitDegrees=45
  swing2Motion=Free
  twistLimitDegrees=45
  twistMotion=Free
 }
 {bone1Name=dyn_hair_tail_01, bone2Name=dyn_parent_hair_tail_01, swing1LimitDegrees=60, swing2LimitDegrees=45, twistLimitDegrees=3}
 {bone1Name=dyn_hair_tail_02, bone2Name=dyn_hair_tail_01, swing1LimitDegrees=60, swing2LimitDegrees=45, twistLimitDegrees=3}
 {bone1Name=dyn_hair_tail_03, bone2Name=dyn_hair_tail_02, swing1LimitDegrees=60, swing2LimitDegrees=45, twistLimitDegrees=3}
 {bone1Name=dyn_hair_tail_04, bone2Name=dyn_hair_tail_03, swing1LimitDegrees=60, swing2LimitDegrees=45, twistLimitDegrees=3}
 {bone1Name=dyn_hair_tail_05, bone2Name=dyn_hair_tail_04, swing1LimitDegrees=60, swing2LimitDegrees=45, twistLimitDegrees=3}
 {bone1Name=dyn_hair_tail_06, bone2Name=dyn_hair_tail_05, swing1LimitDegrees=60, swing2LimitDegrees=45, twistLimitDegrees=3}
 {
  bone1Name=upperarm_l
  bone2Name=spine_05
  swing1LimitDegrees=45
  swing1Motion=Free
  swing2LimitDegrees=45
  swing2Motion=Free
  twistLimitDegrees=45
  twistMotion=Free
 }
 {
  bone1Name=upperarm_r
  bone2Name=spine_05
  swing1LimitDegrees=45
  swing1Motion=Free
  swing2LimitDegrees=45
  swing2Motion=Free
  twistLimitDegrees=45
  twistMotion=Free
 }
 {
  bone1Name=dyn_backpack_col
  bone2Name=spine_05
  swing1LimitDegrees=45
  swing1Motion=Free
  swing2LimitDegrees=45
  swing2Motion=Free
  twistLimitDegrees=45
  twistMotion=Free
 }
]
